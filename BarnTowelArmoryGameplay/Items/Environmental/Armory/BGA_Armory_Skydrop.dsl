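; Skydrop actor for the mobile Armory, which freefalls to a landing position, then is hidden or destroyed once landed.
; Timed height checks trigger thrusters, a timeline eases velocity before a slow final descent, and replicated state changes snap its location.
; A timed capsule trace knocks back players and vehicles, with per-actor cooldowns, and destroys buildings via gameplay effects.
; Building destruction is immediate for close hits when something needs knockback, otherwise queued asynchronously.
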(event EventBeginPlay
  ;[0 1 0.06801] Begin Play
  ; Calculate the initial velocity to be used if resuming on a non-blocking hit
  (|CallParentFunction)
  (Variables|FreefallMovement|SetFreefallSpinTargetRotation (Transformation|GetActorRotation))
  (Variables|Default|SetCachedPawnTraceLength (ScalableFloat|GetValueatLevel (Variables|Collision|GetHFRow_PawnTraceLength)))
  (Variables|Default|SetbCachedDestroysObstacles (ScalableFloat|GetValueasBool (Variables|Collision|GetHFRow_DestoysObstacles)))
  (Rendering|Material|SetScalarParameterValueonMaterials (Variables|Default|GetMobileArmory_Exterior) "BP_GetTime" (Utilities|Time|GetGameTimeinSeconds))
  (Networking|CanExecuteCosmeticEvents
    (:True
      (GameplayCue|ExecuteGameplayCue%28Burst%29Local self (Variables|FX|GameplayCues|GetGCNTag_EntrySpawn) (Ability|GameplayCue|MakeGameplayCueParameters 0.0)))
    (:False
      (Variables|FreefallMovement|SetCurrentVelocity (* (Variables|FreefallMovement|GetFallDirection) (ScalableFloat|GetValueatLevel (Variables|FreefallMovement|GetSF_FallVelocity))))
      (Ability|AddGameplayCueReplicated (Variables|FX|GameplayCues|GetGCNTag_Freefall))
      (Ability|AddGameplayCueReplicated (Variables|FX|GameplayCues|GetGCNTag_ThrustersAdditional))
      (FreefallMovement|SetProjectileMovementVelocity (Variables|FreefallMovement|GetCurrentVelocity))
      (bind _returnvalue (Utilities|Time|SetTimerbyEvent (EventDispatchers|CreateEvent :!event Custom|CheckHeightForThrusters) (ScalableFloat|GetValueatLevel (Variables|FreefallMovement|GetSF_HeightCheckInterval)) true true))
      (Variables|Thrusters|SetThrusterHeightCheckTimerHandle _returnvalue)
      (bind _returnvalue_2 (Utilities|Time|SetTimerbyEvent (EventDispatchers|CreateEvent :!event Custom|HandleCollision) (ScalableFloat|GetValueatLevel (Variables|Collision|GetSF_CollisionCheckFrequency)) true))
      (Variables|Collision|SetCollisionTimer _returnvalue_2)
      (Variables|Default|SetArmoryGameplayState "NewEnumerator1")
      (bind (_direction _rotationalpha) (Timeline|PodTransform PlayFromStart)
        (:Update
          (Networking|CanExecuteCosmeticEvents
            (:True
              (Transformation|SetActorRelativeRotation (Math|Rotator|Lerp(Rotator) (Math|Rotator|MakeRotator (.roll (Variables|FreefallMovement|GetFreefallSpinTargetRotation)) (.pitch (Variables|FreefallMovement|GetFreefallSpinTargetRotation)) (+ (.yaw (Variables|FreefallMovement|GetFreefallSpinTargetRotation)) (Variables|FreefallMovement|GetFreefallSpinRotationAmount))) (Variables|FreefallMovement|GetFreefallSpinTargetRotation) _rotationalpha)))) ; rotation only needed on the client because \ncollision is not handled by the mesh
        )) ; replicate rotation to clients
    )))

(event Custom|CheckHeightForThrusters
  (bind (_returnvalue_x _returnvalue_y _returnvalue_z) (Math|Vector|BreakVector (Transformation|GetActorLocation)))
  (bind (_landingposition_x _landingposition_y _landingposition_z) (Math|Vector|BreakVector (Variables|GetLandingPosition_X)))
  ;[0.3544 0.6771 1] Check the distance from the landing pad to decide when to activate the thrusters.
  (if (<= (- _returnvalue_z _landingposition_z) (ScalableFloat|GetValueatLevel (Variables|FreefallMovement|GetSF_LandingHeight)))
    (CallFunction|BeginLandingThrusters)
    (Utilities|Time|ClearandInvalidateTimerbyHandle (Variables|Thrusters|GetThrusterHeightCheckTimerHandle))))

(event Custom|BeginLandingThrusters
  ;[0.3544 0.6771 1] BeginLandingThrusters
  ; Reduce the velocity over a short timeline to conserve momentum. Then finish landing at a much slower rate. Toggle appropriate FX
  (Variables|Default|SetArmoryGameplayState "NewEnumerator3")
  (Ability|RemoveGameplayCueReplicated (Variables|FX|GameplayCues|GetGCNTag_Freefall))
  (Ability|AddGameplayCueReplicated (Variables|FX|GameplayCues|GetGCNTag_Thrusters))
  (Variables|Thrusters|SetbLandingThrustersActivated true)
  (Components|Timeline|SetPlayRate (Variables|BGA_Armory_Skydrop|GetThrusterVelocityNormalized) (Math|Float|SafeDivide 1.0 (ScalableFloat|GetValueatLevel (Variables|Landing|GetSF_ThrustersDuration))))
  (bind (_direction _velocitycurve) (Timeline|ThrusterVelocityNormalized Play)
    (:Update
      (bind _output_get (Variables|FreefallMovement|SetCurrentVelocity (Math|Vector|Lerp(Vector) (* (Variables|FreefallMovement|GetfallDirection) (ScalableFloat|GetValueatLevel (Variables|FreefallMovement|GetSF_FallVelocity))) "0, 0, 0" _velocitycurve)))
      (FreefallMovement|SetProjectileMovementVelocity _output_get))
    (:Finished
      (if (Utilities|Enum|Equal(Enum) (Variables|Default|GetArmoryGameplayState) "NewEnumerator3")
        (Utilities|FlowControl|Delay (ScalableFloat|GetValueatLevel (Variables|Landing|GetSF_DelayBeforeFinalDescend)))
        (bind _output_get_2 (Variables|FreefallMovement|SetCurrentVelocity (Utilities|Operators|Multiply (Variables|FreefallMovement|GetfallDirection) (ScalableFloat|GetValueatLevel (Variables|FreefallMovement|GetSF_FallVelocity)) (Variables|Thrusters|GetThrusterFallVelocityModifier))))
        (FreefallMovement|SetProjectileMovementVelocity _output_get_2)
        (Networking|SwitchHasAuthority
          (:Authority
            (bind _returnvalue (Utilities|Time|SetTimerbyEvent (EventDispatchers|CreateEvent :!event Custom|CheckForEndLocation) (ScalableFloat|GetValueatLevel (Variables|FreefallMovement|GetSF_HeightCheckInterval)) true true)) ; Begin checking for our landing location
            (Variables|FreefallMovementr|SetCheckForEndLocationTimer _returnvalue)))))))

(event Custom|EndThrusterVelocityManipulation
  ;[0.3544 0.6771 1] BeginLandingThrusters
  ; Reduce the velocity over a short timeline to conserve momentum. Then finish landing at a much slower rate. Toggle appropriate FX
  (Timeline|ThrusterVelocityNormalized Stop))

(event Custom|HandleCollision
  (bind _cachedpawntracelength (Variables|Default|GetCachedPawnTraceLength))
  ;[0.3544 0.6771 1] HandleCollision
  ; The larger capsule that tracks hits while the armory is in freefall should knock players back and prevent overlapping/clipping
  (bind _outhits (Collision|MultiCapsuleTraceForObjects (+ (Transformation|GetActorLocation) (Variables|Collision|GetDestructionTraceOffset)) (Utilities|Operators|Subtract (+ (Transformation|GetActorLocation) (Variables|Collision|GetDestructionTraceOffset)) (Math|Vector|MakeVector 0.0 0.0 (ScalableFloat|GetValueatLevel (Variables|Collision|GetHFRow_BuildingTraceLength)))) (Variables|Collision|GetDestructionTrace_Radius) (Variables|Collision|GetDestructionTrace_HalfHeight) (Utilities|Array|MakeArray "ObjectTypeQuery1" "ObjectTypeQuery2" "ObjectTypeQuery5" "ObjectTypeQuery3" "ObjectTypeQuery7" "ObjectTypeQuery4" "ObjectTypeQuery6" "ObjectTypeQuery1")))
  (for _array_element _outhits
    (Utilities|Casting|CastToBuildingActor (.HitActor (Collision|BreakHitResult _array_element))
      (:then
        _)
      (:CastFailed
        (if (<= (.Distance (Collision|BreakHitResult _array_element)) _cachedpawntracelength)
          (bind _validhit (Collision|TrackHitActorAgainstCoolodown (.HitActor (Collision|BreakHitResult _array_element)))) ; Only for non-buildings.
          (if _validhit
            _))))
    (if (and (Variables|/Script/Engine.Actor|GetbCanBeDamaged (.HitActor (Collision|BreakHitResult _array_element))) (Variables|Default|GetbCachedDestroysObstacles))
      (bind _asbuilding_actor (Utilities|Casting|CastToBuildingActor (.HitActor (Collision|BreakHitResult _array_element)))
        (:then
          (if (and (> (Health|GetMaxHealth _asbuilding_actor) 0.0) (<= (Health|GetMaxHealth _asbuilding_actor) 5e+03))
            (if (<= (.Distance (Collision|BreakHitResult _array_element)) _cachedpawntracelength)
              (Utilities|Array|AddUnique (Variables|Default|GetImmediateHits) _array_element)
              (elif (Utilities|Array|ContainsItem (Variables|Default|GetBuildingsPendingDestruction) (.HitActor (Collision|BreakHitResult _array_element)))
                _
                (else
                  (Utilities|Array|AddUnique (Variables|Default|GetAsyncHits) _array_element))))))
        (:CastFailed
          (Utilities|Array|AddUnique (Variables|Default|GetNonBuildingHits) _array_element)))))
  (if (Utilities|Array|IsNotEmpty (Variables|Default|GetNonBuildingHits))
    ;; Player knockback and instant destruction of buildings that might be in the way.
    (Utilities|FlowControl|Sequence
      (:then_0
        (for _array_element_2 (Variables|Default|GetNonBuildingHits)
          (Collision|KnockbackPlayersorVehicles (.HitActor (Collision|BreakHitResult _array_element_2))))
        (Utilities|Array|Clear (Variables|Default|GetNonBuildingHits)))
      (:then_1
        (for _array_element_3 (Variables|Default|GetImmediateHits)
          (Collision|FreefallDestroyHitActor (.HitActor (Collision|BreakHitResult _array_element_3))))
        (Utilities|Array|Clear (Variables|Default|GetImmediateHits)))
      (:then_2
        _))
    (else
      ;; Queued building destruction
      (Utilities|Array|AppendArray (Variables|Default|GetAsyncHits) (Variables|Default|GetImmediateHits)) ; If we haven't hit a player, close buildings can be queued too.
    ))
  (bind _buildingdestructionasynchandle (Variables|Default|GetBuildingDestructionAsyncHandle)
    (:then
      (if (AsyncAction|IsActive _buildingdestructionasynchandle)
        (Ability|Tasks|AddTargets _buildingdestructionasynchandle (Variables|Default|GetAsyncHits))
        (else
          (bind _asynctaskproxy (Ability|Tasks|ApplyGameplayEffectToHitResultsAsync (Ability|GetAbilitySystemComponent 0) (Variables|Default|GetAsyncHits) "/Game/Athena/SpecialActors/Skydrop/GE_Skydrop_Environment_Destruction.GE_Skydrop_Environment_Destruction_C" 0 1 (Class|AthenaFunctionLibrary|MakeSimpleFortEffectContext self self) (Variables|Collision|GetHFRow_AsyncDestructionBuildingsPerTick)))
          (Variables|Default|SetBuildingDestructionAsyncHandle _asynctaskproxy)))
      ;; Keep track of buildings pending destruction
      (for _array_element_4 (Variables|Default|GetAsyncHits)
        (Utilities|Array|AddUnique (Variables|Default|GetBuildingsPendingDestruction) (.HitActor (Collision|BreakHitResult _array_element_4))))
      (Utilities|Array|Clear (Variables|Default|GetAsyncHits))
      (Utilities|Array|Clear (Variables|Default|GetImmediateHits)))
    (:else
      _)))

(event Custom|ArmoryLanded
  ;[0 1 0.06801] ArmoryLanded (SERVER)
  ; Set the landed state, load the appropriate Data Layer for the interior level design, and start the delay timer before the doors are opened
  (FreefallMovement|SetProjectileMovementVelocity "0, 0, 0")
  (Components|Movement|StopMovementImmediately (Variables|Default|GetFortProjectileMovement))
  (Ability|RemoveGameplayCueReplicated (Variables|FX|GameplayCues|GetGCNTag_Thrusters))
  (Ability|RemoveGameplayCueReplicated (Variables|FX|GameplayCues|GetGCNTag_ThrustersAdditional)))

(event Custom|ArmoryLanded_Shared
  ;[0.1299 0.1558 1] Set location on client and server when we land
  (CallFunction|EndThrusterVelocityManipulation)
  (Transformation|SetActorLocation (Variables|Landing|GetLandingPosition) false true))

(event Custom|CheckForEndLocation
  (bind (_returnvalue_x _returnvalue_y _returnvalue_z) (Math|Vector|BreakVector (Transformation|GetActorLocation)))
  (bind (_landingposition_x _landingposition_y _landingposition_z) (Math|Vector|BreakVector (Variables|GetLandingPosition_X)))
  ;[0.2059 0.9403 1] Check if we're close enough to our landing position, rather than waiting for a collision with the pad.
  (if (<= (- _returnvalue_z _landingposition_z) 1.0)
    (Variables|Default|SetArmoryGameplayState "NewEnumerator4")
    (Utilities|Time|ClearandInvalidateTimerbyHandle (Variables|FreefallMovementr|GetCheckForEndLocationTimer))))

(event Custom|HideLandedArmory_Client
  ;[1 0.367 0] DestroyLandedArmory
  ; Remove the more expensive animated Armory with a static data layer
  (Rendering|SetVisibility (Variables|Default|GetMobileArmory_Exterior) false true)
  (Collision|SetActorEnableCollision false)
  (Utilities|Time|ClearandInvalidateTimerbyHandle (Variables|Collision|GetCollisionTimer)))

(event Custom|DestroyLandedArmory_Server
  (bind _hfrow_delaybeforedestructiononlanded (Variables|Default|GetHFRow_DelayBeforeDestructionOnLanded))
  (bind _returnvalue (ScalableFloat|GetValueatLevel _hfrow_delaybeforedestructiononlanded))
  ;[1 0.367 0] DestroyLandedArmory
  ; Remove the more expensive animated Armory with a static data layer
  (Utilities|FlowControl|Sequence
    (:then_0
      (Collision|SetActorEnableCollision false)
      (Utilities|Time|ClearandInvalidateTimerbyHandle (Variables|Collision|GetCollisionTimer)))
    (:then_1
      (if (> _returnvalue 0)
        (Actor|SetLifeSpan _returnvalue)
        (else
          (Actor|DestroyActor))))))

(fn FreefallDestroyHitActor
  (params (HitActor :type Actor%20Object%20Reference))
  (GameplayEffects|ApplyGameplayEffectToTarget :self (Ability|GetAbilitySystemComponent self) :GameplayEffectClass "/Game/Athena/SpecialActors/Skydrop/GE_Skydrop_Environment_Destruction.GE_Skydrop_Environment_Destruction_C" :Target (Ability|GetAbilitySystemComponent HitActor) :Level 0.0))

(fn HitActorCooldownTimerExpired
  (bind _hitactorcooldownbyactormap (Variables|Collision|GetHitActorCooldownByActorMap))
  (bind _hnewcooldowntimerduration (LocalVariables|GethNewCooldownTimerDuration))
  ;[0.2567 0.4925 0.7656] HitActorCooldownTimerExpired
  ; The cooldown has expired, remove all expired actors to accept new hits, set new cooldown timer if needed
  (Utilities|Time|ClearandInvalidateTimerbyHandle (Variables|Collision|GetTimer_HitActorCooldown))
  (Utilities|FlowControl|Sequence
    (:then_0
      ;;[1 0.5108 0.01025] Remove all actors that have become null
      (Utilities|Map|Remove _hitactorcooldownbyactormap)
      (LocalVariables|SethNewCooldownTimerDuration 0.0) ; reset next cooldown timer duration
    )
    (:then_1
      (bind _output_get (LocalVariables|SethCurrentCooldownElapsedTime (- (Fortnite|Sync|Time|GetServerWorldTimeSeconds) (Variables|Collision|GetHitActorRefreshTime_Const))))
      (bind _keys (Utilities|Map|Keys _hitactorcooldownbyactormap))
      (bind _values (Utilities|Map|Values _hitactorcooldownbyactormap))
      ;;[0.2567 0.4925 0.7656] All hit and ignored actors that HitCooldown has expired should be removed and hit functionality restored
      (for (_array_element _array_index) _values
        (if (< _array_element _output_get)
          (Utilities|Array|Add (LocalVariables|GetActorsToRemove) (Utilities|Array|Get(acopy) _keys _array_index))
          (else
            ;;[0.5724 1 0.6015] If there's any valid timers remaining, set the next cooldown clock to manage those actors
            (if (< _array_element (LocalVariables|GethNewCooldownTimerDuration))
              (LocalVariables|SethNewCooldownTimerDuration _array_element))))))
    (:then_2
      (for _array_element_2 (LocalVariables|GetActorsToRemove)
        ;;[0.2567 0.4925 0.7656] Remove actors whose cooldown has been exceeded
        (Utilities|Map|Remove (Variables|Collision|GetHitActorCooldownByActorMap) _array_element_2))
      ;;[0.4271 0.2466 0.08268] Set the next cooldown timer
      (if (> _hnewcooldowntimerduration 0)
        (bind _returnvalue (Utilities|Time|SetTimerbyEvent 0 _hnewcooldowntimerduration))
        (Variables|Collision|SetTimer_HitActorCooldown _returnvalue)))))

(fn KnockbackPlayersOrVehicles
  (params (HitActor :type Actor%20Object%20Reference))
  (bind _self self)
  (bind _returnvalue (Ability|GetAbilitySystemComponent _self))
  ;[0.4308 0.6721 1] Launch characters and vehicles if we run into them
  (Networking|SwitchHasAuthority
    (:Authority
      (Utilities|IsValid _returnvalue
        (:Is%20Valid
          (GameplayEffects|ApplyGameplayEffectToTarget _returnvalue (Variables|Collision|GetGE_Knockback) (Ability|GetAbilitySystemComponent HitActor) 0.0 (Class|AthenaFunctionLibrary|MakeSimpleFortEffectContext self self)))))))

(fn OnRep_ArmoryGameplayState
  (bind (_returnvalue_x _returnvalue_y) (Math|Vector|BreakVector (Transformation|GetActorLocation)))
  (bind _armorygameplaystate (Variables|Default|GetArmoryGameplayState))
  (bind (_landingposition_x _landingposition_y _landingposition_z) (Math|Vector|BreakVector (Variables|GetLandingPosition_X)))
  (Networking|SwitchHasAuthority
    (:Authority
      (Default|CallOnArmoryStateChanged _armorygameplaystate))
    (:Remote
      _))
  (switch Utilities|FlowControl|Switch|SwitchonEnum_BarnTowelArmory_LandingState _armorygameplaystate
    (:NewEnumerator3
      (Transformation|SetActorLocation 0 (Math|Vector|MakeVector _returnvalue_x _returnvalue_y (+ _landingposition_z (ScalableFloat|GetValueatLevel (Variables|FreefallMovement|GetSF_LandingHeight)))) false true))
    (:NewEnumerator4
      (Networking|SwitchHasAuthority
        (:Authority
          (CallFunction|ArmoryLanded))
        (:Remote
          _))
      (CallFunction|ArmoryLandedShared))))

(fn SetProjectileMovementVelocity
  (params (NewVelocity :type Vector))
  (Projectile|SetVelocity (Variables|Default|GetFortProjectileMovement) NewVelocity))

(fn TrackHitActorAgainstCoolodown
  (params (HitActor :type Actor%20Object%20Reference))
  (bind _hitactorcooldownbyactormap (Variables|Collision|GetHitActorCooldownByActorMap))
  (bind (_value _returnvalue) (Utilities|Map|Find _hitactorcooldownbyactormap HitActor))
  (bind _returnvalue_2 (Fortnite|Sync|Time|GetServerWorldTimeSeconds))
  ;[0.2567 0.4925 0.7656] TrackHitActorAgainstCooldown
  ; Check current actors that were previously hit - those will be ignored, otherwise allow the hit event and add it to tracking
  (if (and (>= (- _returnvalue_2 _value) (Variables|Collision|GetHitActorRefreshTime_Const)) (not _returnvalue))
    (Utilities|Map|Add _hitactorcooldownbyactormap HitActor _returnvalue_2)
    (if (not (Utilities|Time|IsTimerActivebyHandle (Variables|Collision|GetTimer_HitActorCooldown)))
      (bind _returnvalue_3 (Utilities|Time|SetTimerbyEvent (EventDispatchers|CreateEvent :!event HitActorCooldownTimerExpired) (Variables|Collision|GetHitActorRefreshTime_Const))) ; no current timer is active, set a new cooldown timer to maximum time
      (Variables|Collision|SetTimer_HitActorCooldown _returnvalue_3))
    ;;[1 0.5212 0.8873] RETURN (True)
    ;; Hit Actor is not yet on cooldown
    ;; or it has exceeded the timer
    (return true)
    (else
      ;;[1 0.5212 0.8873] RETURN (False)
      ;; Actor Already hit and on valid cooldown
      (return false))))

(fn ConstructionScript)
