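; Marker Blueprint for the BarnTowel armory that marks the armory for all players when play begins.
; GetMarkerDetailsId builds the marker details Name by appending the selected armory variant's enum string to the "MarkerDetails.BarnTowelArmory." prefix.
; A Select on the variant supplies that same built ID for two enumerators, with any other options left at their defaults.
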(event EventBeginPlay
  (MarkableActor|MarkArmoryforAllPlayers))

(fn GetMarkerDetailsId
  (return (Utilities|String|StringToName (Utilities|Select :Index (Variables|Default|GetSelectedArmoryVariant) :NewEnumerator2 (Utilities|String|Append "MarkerDetails.BarnTowelArmory." (Utilities|String|EnumtoString (Variables|Default|GetSelectedArmoryVariant))) :NewEnumerator1 (Utilities|String|Append "MarkerDetails.BarnTowelArmory." (Utilities|String|EnumtoString (Variables|Default|GetSelectedArmoryVariant)))))))

(fn ConstructionScript)
