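; Armory interior light prop that switches on when its reactive prop component reports the Armory opened.
; The open event sets a RepNotify bool, replicating the activated state to clients.
; Its OnRep checks that cosmetic events can execute, then shows the light cone billboard and plays a power-on sound at the actor's location.
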(event OnArmoryOpened(BP_BarnTowelArmory_ReactiveProp_Component)
  ;[0 0.1592 1] OnArmoryOpened
  ; Set a RepNotify variable to replicate to clients for cosmetic updates
  (Variables|Default|SetbArmoryOpenedAndActivated true))

(event Custom|ActivateOpenedVisuals
  ;[0.4027 0 1] ActivateOpenedVisuals (Client Only)
  ; Set visibility on and play localized sound for the lights turning on to make the Armory feel more responsive
  (Rendering|SetVisibility (Variables|Default|GetS_Billboard_LightCone) true true)
  (Audio|PlaySoundatLocation "/Game/Sounds/Creative/Toys/ActionTrigger/ActionTrigger_PowerOn_Cue.ActionTrigger_PowerOn_Cue" (Transformation|GetActorLocation)))

(fn OnRep_bArmoryOpenedAndActivated
  ;[0.2921 0.5019 1] OnRep ArmoryOpenedAndActivated
  ; Replicated variable to send the activation event to clients for any logic needed
  (if (Variables|Default|GetbArmoryOpenedAndActivated)
    (Networking|CanExecuteCosmeticEvents
      (:True
        (CallFunction|ActivateOpenedVisuals)))))

(fn ConstructionScript)
