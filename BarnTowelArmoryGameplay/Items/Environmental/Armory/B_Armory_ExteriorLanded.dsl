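; Landed exterior of the BarnTowel Armory, which opens its doors and tracks players in its area.
; On BeginPlay the server applies hotfix-driven build blocking and no-build collision.
; The manager's opening event sets RepNotify bIsReadyToOpen: cosmetic instances animate doors through material WPO with audio, while non-cosmetic ones run a timer that reports the opened state.
; The overlap volume adds or removes a POI tag and notifies the manager on entry.
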(event EventBeginPlay
  (bind _hfrow_disableplayerbuildingonlanded (Variables|Default|GetHFRow_DisablePlayerBuildingOnLanded))
  (bind _returnvalue (ScalableFloat|GetValueasBool _hfrow_disableplayerbuildingonlanded))
  ;[0 1 0.2152] BeginPlay
  ; Set the default state of the landed Armory
  (|CallParentFunction)
  (Rendering|Material|SetMaterial (Variables|Debug|GetStaticMeshComponent) 3 (Variables|Default|GetPrimaryVariantMaterial)) ; needs to set this for... reasons\notherwise armory is invisible
  (Networking|SwitchHasAuthority
    (:Authority
      ;;[0.4987 0.7268 1] Setup the Build rules based on the hotfix row
      ;; - this asset only needs to deal with the "On Landed" row since it not activated until the Armory has landed
      (Variables|Building|SetbForceBlockBuildings _returnvalue) ; blocks player building for overlapping structures
      (Collision|SetCollisionEnabled (Variables|Default|GetNoBuildZone) (select _returnvalue "QueryOnly" "NoCollision")) ; "No Build Zones" handle deployables, \n(such as Port-A-Bunkers)
    )))

(event OnArmoryOpening(BP_BarnTowelArmory_ReactiveProp_Component)
  ;[0.3594 0.1904 0.06468] OnArmoryOpening
  ; Event recieved from the manager - replicate to clients using RepNotify
  (Variables|Default|SetbIsReadyToOpen true))

(event Custom|AnimateOpeningArmory
  ;[0.1211 0 1] AnimateOpeningArmory
  ; Open the armory - the interior should have already spawned to avoid visual popping
  (Utilities|FlowControl|Delay (ScalableFloat|GetValueatLevel (Variables|Default|GetSF_OpeningDelay))) ; just a short delay (steam VFX on door, etc)
  (CallFunction|DoorsAnimationWPO)
  (Audio|Components|Audio|Play (Variables|Default|GetAudio_DoorRear) 0.0))

(event Custom|SetOpeningTimeOnServer
  ;[0.2699 0.4876 1] SetOpeningTimeOnServer
  ; The server does not need to animate, just needs to know how long the animation will take
  (Utilities|Time|SetTimerbyEvent (EventDispatchers|CreateEvent :!event Custom|ArmoryOpeningAnimationCompleted) (+ (ScalableFloat|GetValueatLevel (Variables|Default|GetSF_OpeningDelay)) (ScalableFloat|GetValueatLevel (Variables|Default|GetSF_DoorOpeningDuration)))))

(event Custom|DoorsAnimation_WPO
  (Rendering|Material|SetScalarParameterValueonMaterials (Variables|Default|GetMobileArmory_FrontDoor) "BP_GetTime" (Utilities|Time|GetGameTimeinSeconds))
  ; DOORS?
  (Audio|Components|Audio|Play (Variables|Default|GetAudio_DoorFront) 0.0) ; Delay set in sound cue
)

(event Custom|ArmoryOpeningAnimationCompleted
  ;[0.02214 0.342 0.1684] ArmoryOpeningAnimationCompleted
  ; When the armory is fully opened, set the manager State to OPENED
  (Variables|Default|SetbIsArmoryOpened true)
  (bind _owningarmoryspawnmanager (Variables|/BarnTowelArmoryGameplay/Items/BP_BarnTowelArmory_ReactiveProp_Component.BP_BarnTowelArmory_ReactiveProp_Component_C|GetOwningArmorySpawnManager (Variables|Default|GetBP_BarnTowelArmory_ReactiveProp_Component))
    (:then
      (Class|BBarnTowelSpawnManager|SetNewArmoryState _owningarmoryspawnmanager "NewEnumerator6") ; once the opening animations are complete, \ntrigger the Armory final idle state
    )
    (:else)))

(event OnComponentBeginOverlap(OverlapVolume)
  (params (OverlappedComponent :type Primitive%20Component%20Object%20Reference) (OtherActor :type Actor%20Object%20Reference) (OtherComp :type Primitive%20Component%20Object%20Reference) (OtherBodyIndex :type Integer) (bFromSweep :type Boolean) (SweepResult :type Hit%20Result%20Structure%20%28by%20ref%29))
  (Ability|GameplayEffect|AddLooseGameplayTags OtherActor (Variables|Default|GetBarnTowelPOITag))
  (bind _asfort_pawn (Utilities|Casting|CastToFortPawn OtherActor))
  (Class|BBarnTowelSpawnManager|PlayerEntered (Variables|/BarnTowelArmoryGameplay/Items/BP_BarnTowelArmory_ReactiveProp_Component.BP_BarnTowelArmory_ReactiveProp_Component_C|GetOwningArmorySpawnManager (Variables|Default|GetBP_BarnTowelArmory_ReactiveProp_Component)) _asfort_pawn))

(event OnComponentEndOverlap(OverlapVolume)
  (params (OverlappedComponent :type Primitive%20Component%20Object%20Reference) (OtherActor :type Actor%20Object%20Reference) (OtherComp :type Primitive%20Component%20Object%20Reference) (OtherBodyIndex :type Integer))
  (Ability|GameplayEffect|RemoveLooseGameplayTags OtherActor (Variables|Default|GetBarnTowelPOITag)))

(fn OnRep_bIsReadyToOpen
  (if (not (Variables|Default|GetbIsArmoryOpened))
    (if (Variables|Default|GetbIsReadyToOpen)
      (Networking|CanExecuteCosmeticEvents
        (:True
          (CallFunction|AnimateOpeningArmory))
        (:False
          (CallFunction|SetOpeningTimeonServer))))
    (else
      (Rendering|Material|SetScalarParameterValueonMaterials (Variables|Default|GetMobileArmory_FrontDoor) "BP_GetTime" 0.0) ; setting the armory to its already opened animation end state
    )))

(fn ConstructionScript)
